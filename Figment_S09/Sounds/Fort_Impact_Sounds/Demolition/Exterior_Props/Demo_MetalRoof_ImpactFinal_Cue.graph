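# Sound cue for the final demolition impact on metal roof exterior props, playing metal roof death sounds.
# A quality-level node picks one of four branches, offering three random waves, two random waves, or a single fixed wave in the last two.
# Each branch applies the same slight random pitch modulation.
# It uses close demolition attenuation, a ducking impact sound class, and an object-impact concurrency set.
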SoundCue /Figment_S09/Sounds/Fort_Impact_Sounds/Demolition/Exterior_Props/Demo_MetalRoof_ImpactFinal_Cue
root=QualityLevel_0
nodes [
 {id=Modulator_0, children [Random_0], properties {pitchMax=1.07, pitchMin=0.92}}
 {id=Modulator_1, children [WavePlayer_3], properties {pitchMax=1.07, pitchMin=0.92}}
 {id=Modulator_2, children [Random_18], properties {pitchMax=1.07, pitchMin=0.92}}
 {id=Modulator_3, children [WavePlayer_6], properties {pitchMax=1.07, pitchMin=0.92}}
 {id=QualityLevel_0, children [Modulator_2,Modulator_0,Modulator_1,Modulator_3]}
 {id=Random_0, children [WavePlayer_0,WavePlayer_2], properties {weights [1,1]}}
 {id=Random_18, children [WavePlayer_60, WavePlayer_61, WavePlayer_24], properties {weights [1,1,1]}}
 {id=WavePlayer_0, properties {soundWaveAssetPtr=/Figment_S09/Sounds/Fort_Impact_Sounds/Demolition/Metal_Roof_Death02}}
 {id=WavePlayer_2, properties {soundWaveAssetPtr=/Figment_S09/Sounds/Fort_Impact_Sounds/Demolition/Metal_Roof_Death04}}
 {id=WavePlayer_24, properties {soundWaveAssetPtr=&Metal_Roof_Death04}}
 {id=WavePlayer_3, properties {soundWaveAssetPtr=&Metal_Roof_Death02}}
 {id=WavePlayer_6, properties {soundWaveAssetPtr=&Metal_Roof_Death02}}
 {id=WavePlayer_60, properties {soundWaveAssetPtr=&Metal_Roof_Death02}}
 {id=WavePlayer_61, properties {soundWaveAssetPtr=/Figment_S09/Sounds/Fort_Impact_Sounds/Demolition/Metal_Roof_Death03}}
]
properties {
 attenuationSettings=/Figment_S09/Sounds/Attenuation/Fort_Demolition_Close_Att
 maxDistance=2000
 soundClassObject=/Game/Sounds/FortSoundClassesAndModes/Fort_ImpactSounds_DuckDefault_PSM
 volumeMultiplier=0.8
 concurrencyOverrides {resolutionRule=PreventNew}
 concurrencySet [&CONC_Object_Impacts_FourMax]
}
