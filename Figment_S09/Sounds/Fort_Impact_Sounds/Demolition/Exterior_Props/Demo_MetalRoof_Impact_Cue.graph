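# Sound cue for metal roof impact sounds in demolition of exterior props.
# A quality-level switch sends its first two slots to a random pick of three impact waves and its last two to a pick of two.
# Each branch runs through a modulator that slightly varies pitch.
# Playback uses close demolition attenuation, a ducking impact sound class and concurrency settings that limit simultaneous impacts.
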SoundCue /Figment_S09/Sounds/Fort_Impact_Sounds/Demolition/Exterior_Props/Demo_MetalRoof_Impact_Cue
root=QualityLevel_0
nodes [
 {id=Modulator_0, children [Random_0], properties {pitchMax=1.07, pitchMin=0.92}}
 {id=Modulator_2, children [Random_15], properties {pitchMax=1.07, pitchMin=0.92}}
 {id=QualityLevel_0, children [Modulator_2,Modulator_2,Modulator_0,Modulator_0]}
 {id=Random_0, children [WavePlayer_0,WavePlayer_1], properties {weights [1,1]}}
 {id=Random_15, children [WavePlayer_54, WavePlayer_55, WavePlayer_7], properties {weights [1,1,1]}}
 {id=WavePlayer_0, properties {soundWaveAssetPtr=/Figment_S09/Sounds/Fort_Impact_Sounds/Demolition/Metal_Roof_Impact03}}
 {id=WavePlayer_1, properties {soundWaveAssetPtr=/Figment_S09/Sounds/Fort_Impact_Sounds/Demolition/Metal_Roof_Impact02}}
 {id=WavePlayer_54, properties {soundWaveAssetPtr=&Metal_Roof_Impact02}}
 {id=WavePlayer_55, properties {soundWaveAssetPtr=&Metal_Roof_Impact03}}
 {id=WavePlayer_7, properties {soundWaveAssetPtr=/Figment_S09/Sounds/Fort_Impact_Sounds/Demolition/Metal_Roof_Impact05}}
]
properties {
 attenuationSettings=/Figment_S09/Sounds/Attenuation/Fort_Demolition_Close_Att
 maxDistance=2000
 soundClassObject=/Game/Sounds/FortSoundClassesAndModes/Fort_ImpactSounds_DuckDefault_PSM
 volumeMultiplier=0.5
 concurrencyOverrides {maxCount=3, resolutionRule=StopOldest}
 concurrencySet [&CONC_Object_Impacts_FourMax]
}
